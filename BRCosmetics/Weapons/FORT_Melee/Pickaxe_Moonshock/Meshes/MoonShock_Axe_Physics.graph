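# Physics asset for the Moonshock pickaxe mesh, giving its hammerhead nine two-bone goo appendages as physics bodies.
# Each small base_goo capsule joins the central base_hammerhead body through a constraint with free swing and twist.
# Each larger goo capsule hangs off its base with very narrow swing and zero twist limits, so rotational freedom sits almost entirely at the base joint.
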PhysicsAsset /BRCosmetics/Weapons/FORT_Melee/Pickaxe_Moonshock/Meshes/MoonShock_Axe_Physics
bodies {
 base_goo_1 [{shapeType=Capsule, length=1, radius=1, shapeName=base_goo_1_capsule, rotation=<rot 90 0 0>}]
 base_goo_2 [{shapeType=Capsule, length=1, radius=1, shapeName=base_goo_2_capsule, rotation=<rot 90 0 0>}]
 base_goo_3 [{shapeType=Capsule, length=1, radius=1, shapeName=base_goo_3_capsule, rotation=<rot 90 0 0>}]
 base_goo_4 [{shapeType=Capsule, length=1, radius=1, shapeName=base_goo_4_capsule, rotation=<rot 90 0 0>}]
 base_goo_5 [{shapeType=Capsule, length=1, radius=1, shapeName=base_goo_5_capsule, rotation=<rot 90 0 0>}]
 base_goo_6 [{shapeType=Capsule, length=1, radius=1, shapeName=base_goo_6_capsule, rotation=<rot 90 0 0>}]
 base_goo_7 [{shapeType=Capsule, length=1, radius=1, shapeName=base_goo_7_capsule, rotation=<rot 90 0 0>}]
 base_goo_8 [{shapeType=Capsule, length=1, radius=1, shapeName=base_goo_8_capsule, rotation=<rot 90 0 0>}]
 base_goo_9 [{shapeType=Capsule, length=1, radius=1, shapeName=base_goo_9_capsule, rotation=<rot 90 0 0>}]
 base_hammerhead [{shapeType=Capsule, length=1, radius=1, shapeName=base_hammerhead_capsule, rotation=<rot 90 0 0>}]
 goo_1 [{shapeType=Capsule, length=8, radius=5, shapeName=goo_1_capsule, rotation=<rot 90 0 0>}]
 goo_2 [{shapeType=Capsule, length=8, radius=5, shapeName=goo_2_capsule, rotation=<rot 90 0 0>}]
 goo_3 [{shapeType=Capsule, length=8, radius=5, shapeName=goo_3_capsule, rotation=<rot 90 0 0>}]
 goo_4 [{shapeType=Capsule, length=8, radius=5, shapeName=goo_4_capsule, rotation=<rot 90 0 0>}]
 goo_5 [{shapeType=Capsule, length=8, radius=5, shapeName=goo_5_capsule, rotation=<rot 90 0 0>}]
 goo_6 [{shapeType=Capsule, length=8, radius=5, shapeName=goo_6_capsule, rotation=<rot 90 0 0>}]
 goo_7 [{shapeType=Capsule, length=8, radius=5, shapeName=goo_7_capsule, rotation=<rot 90 0 0>}]
 goo_8 [{shapeType=Capsule, length=8, radius=5, shapeName=goo_8_capsule, rotation=<rot 90 0 0>}]
 goo_9 [{shapeType=Capsule, length=8, radius=5, shapeName=goo_9_capsule, rotation=<rot 90 0 0>}]
}
constraints [
 {
  bone1Name=base_goo_1
  bone2Name=base_hammerhead
  swing1LimitDegrees=45
  swing1Motion=Free
  swing2LimitDegrees=45
  swing2Motion=Free
  twistLimitDegrees=45
  twistMotion=Free
 }
 {bone1Name=goo_1, bone2Name=base_goo_1, swing1LimitDegrees=2, swing2LimitDegrees=2, twistLimitDegrees=0}
 {
  bone1Name=base_goo_2
  bone2Name=base_hammerhead
  swing1LimitDegrees=45
  swing1Motion=Free
  swing2LimitDegrees=45
  swing2Motion=Free
  twistLimitDegrees=45
  twistMotion=Free
 }
 {bone1Name=goo_2, bone2Name=base_goo_2, swing1LimitDegrees=2, swing2LimitDegrees=2, twistLimitDegrees=0}
 {
  bone1Name=base_goo_3
  bone2Name=base_hammerhead
  swing1LimitDegrees=45
  swing1Motion=Free
  swing2LimitDegrees=45
  swing2Motion=Free
  twistLimitDegrees=45
  twistMotion=Free
 }
 {bone1Name=goo_3, bone2Name=base_goo_3, swing1LimitDegrees=2, swing2LimitDegrees=2, twistLimitDegrees=0}
 {
  bone1Name=base_goo_4
  bone2Name=base_hammerhead
  swing1LimitDegrees=45
  swing1Motion=Free
  swing2LimitDegrees=45
  swing2Motion=Free
  twistLimitDegrees=45
  twistMotion=Free
 }
 {bone1Name=goo_4, bone2Name=base_goo_4, swing1LimitDegrees=2, swing2LimitDegrees=2, twistLimitDegrees=0}
 {
  bone1Name=base_goo_5
  bone2Name=base_hammerhead
  swing1LimitDegrees=45
  swing1Motion=Free
  swing2LimitDegrees=45
  swing2Motion=Free
  twistLimitDegrees=45
  twistMotion=Free
 }
 {bone1Name=goo_5, bone2Name=base_goo_5, swing1LimitDegrees=2, swing2LimitDegrees=2, twistLimitDegrees=0}
 {
  bone1Name=base_goo_6
  bone2Name=base_hammerhead
  swing1LimitDegrees=45
  swing1Motion=Free
  swing2LimitDegrees=45
  swing2Motion=Free
  twistLimitDegrees=45
  twistMotion=Free
 }
 {bone1Name=goo_6, bone2Name=base_goo_6, swing1LimitDegrees=2, swing2LimitDegrees=2, twistLimitDegrees=0}
 {
  bone1Name=base_goo_7
  bone2Name=base_hammerhead
  swing1LimitDegrees=45
  swing1Motion=Free
  swing2LimitDegrees=45
  swing2Motion=Free
  twistLimitDegrees=45
  twistMotion=Free
 }
 {bone1Name=goo_7, bone2Name=base_goo_7, swing1LimitDegrees=2, swing2LimitDegrees=2, twistLimitDegrees=0}
 {
  bone1Name=base_goo_8
  bone2Name=base_hammerhead
  swing1LimitDegrees=45
  swing1Motion=Free
  swing2LimitDegrees=45
  swing2Motion=Free
  twistLimitDegrees=45
  twistMotion=Free
 }
 {bone1Name=goo_8, bone2Name=base_goo_8, swing1LimitDegrees=2, swing2LimitDegrees=2, twistLimitDegrees=0}
 {
  bone1Name=base_goo_9
  bone2Name=base_hammerhead
  swing1LimitDegrees=45
  swing1Motion=Free
  swing2LimitDegrees=45
  swing2Motion=Free
  twistLimitDegrees=45
  twistMotion=Free
 }
 {bone1Name=goo_9, bone2Name=base_goo_9, swing1LimitDegrees=2, swing2LimitDegrees=2, twistLimitDegrees=0}
]
